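# Sound cue for the explosion of the Paprika Fizz item, routed to the Fort_Impacts sound class.
# A mixer layers two branches at equal volume, one picking from eight explode waves and the other from eight impact waves.
# Each branch runs an equally weighted random selection through its own modulator for variation.
# The cue overrides attenuation with its own radius and falloff distance.
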SoundCue /Game/Athena/Sounds/Items/Fizz/Paprika_Fizz_Explode_Cue
root=Mixer_0
nodes [
 {id=Mixer_0, children [Modulator_1,Modulator_0], properties {inputVolume [1,1]}}
 {id=Modulator_0, children [Random_0]}
 {id=Modulator_1, children [Random_1]}
 {
  id=Random_0
  children [WavePlayer_0,WavePlayer_1,WavePlayer_2,WavePlayer_3,WavePlayer_4,WavePlayer_5,WavePlayer_6,WavePlayer_7]
  properties {weights [1,1,1,1,1,1,1,1]}
 }
 {
  id=Random_1
  children [WavePlayer_8, WavePlayer_9, WavePlayer_10, WavePlayer_11, WavePlayer_12, WavePlayer_13, WavePlayer_14, WavePlayer_15]
  properties {weights [1,1,1,1,1,1,1,1]}
 }
 {id=WavePlayer_0, properties {soundWaveAssetPtr=&/Item_Paprika_Fizz_Explode_08}}
 {id=WavePlayer_1, properties {soundWaveAssetPtr=&/Item_Paprika_Fizz_Explode_07}}
 {id=WavePlayer_10, properties {soundWaveAssetPtr=&/Item_Paprika_Fizz_Impact_03}}
 {id=WavePlayer_11, properties {soundWaveAssetPtr=&/Item_Paprika_Fizz_Impact_04}}
 {id=WavePlayer_12, properties {soundWaveAssetPtr=&/Item_Paprika_Fizz_Impact_05}}
 {id=WavePlayer_13, properties {soundWaveAssetPtr=&/Item_Paprika_Fizz_Impact_06}}
 {id=WavePlayer_14, properties {soundWaveAssetPtr=&/Item_Paprika_Fizz_Impact_07}}
 {id=WavePlayer_15, properties {soundWaveAssetPtr=&/Item_Paprika_Fizz_Impact_08}}
 {id=WavePlayer_2, properties {soundWaveAssetPtr=&/Item_Paprika_Fizz_Explode_06}}
 {id=WavePlayer_3, properties {soundWaveAssetPtr=&/Item_Paprika_Fizz_Explode_05}}
 {id=WavePlayer_4, properties {soundWaveAssetPtr=&/Item_Paprika_Fizz_Explode_04}}
 {id=WavePlayer_5, properties {soundWaveAssetPtr=&/Item_Paprika_Fizz_Explode_03}}
 {id=WavePlayer_6, properties {soundWaveAssetPtr=&/Item_Paprika_Fizz_Explode_02}}
 {id=WavePlayer_7, properties {soundWaveAssetPtr=&/Item_Paprika_Fizz_Explode_01}}
 {id=WavePlayer_8, properties {soundWaveAssetPtr=&/Item_Paprika_Fizz_Impact_01}}
 {id=WavePlayer_9, properties {soundWaveAssetPtr=&/Item_Paprika_Fizz_Impact_02}}
]
properties {
 bOverrideAttenuation=true
 maxDistance=3000
 soundClassObject=&Fort_Impacts
 volumeMultiplier=0.8
 attenuationOverrides {falloffDistance=2400, attenuationShapeExtents {x=600}}
}
